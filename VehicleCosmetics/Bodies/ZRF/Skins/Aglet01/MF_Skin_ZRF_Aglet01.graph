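# Material function for the Aglet01 vehicle skin on the ZRF body, producing an animated base color and matching roughness.
# Three differently tiled samples of a shared blurred noise texture scroll along UV1 and multiply into a mask blending dark red with the skin texture.
# A second lerp weighted by the texture itself preserves its bright areas.
# Roughness derives from the desaturated result, overriding it and base color on the shared vehicle body base attributes.
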MaterialFunction /VehicleCosmetics/Bodies/ZRF/Skins/Aglet01/MF_Skin_ZRF_Aglet01
graph {nodes [
  n1 = TextureSample() {texture=&/T_Aglet01_BC}
  n2 = Constant2Vector() {desc=Tile, g=5, r=1}
  n3 = TextureCoordinate() {coordinateIndex=1, desc=Shared TexCoord}
  n4 = Panner(Coordinate=Add(A=Multiply(A=@n2, B=@n3)) {constB=0.5, desc=Offset}) {speedY=0.5}
  n5 = OneMinus(None=ConstantBiasScale(None=TextureSample(UVs=@n4) {samplerType=SAMPLERTYPE_Masks, texture=/VehicleCosmetics/SharedTextures/T_Universal_Noise_Heightmap_30_Blurred}.RGB) {bias=-0.85, scale=5})
  n6 = Panner(Coordinate=Multiply(A=Constant2Vector() {desc=Tile, g=4, r=2}, B=@n3)) {speedY=1.5}
  n7 = OneMinus(None=ConstantBiasScale(None=TextureSample(UVs=@n6) {samplerType=SAMPLERTYPE_Masks, texture=&T_Universal_Noise_Heightmap_30_Blurred}.RGB) {bias=-0.85, scale=5})
  n8 = Multiply(A=Clamp(None=Saturate(None=@n5)) {minDefault=0.9}, B=Clamp(None=Saturate(None=@n7)) {minDefault=0.6})
  n9 = Add(A=Multiply(A=Constant2Vector() {desc=Tile, g=0.3, r=2}, B=@n3)) {constB=0.6, desc=Offset}
  n10 = ConstantBiasScale(None=TextureSample(UVs=Panner(Coordinate=@n9) {speedY=0.5}) {samplerType=SAMPLERTYPE_Masks, texture=&T_Universal_Noise_Heightmap_30_Blurred}.RGB) {bias=-0.9, scale=5}
  n11 = Clamp(None=Saturate(None=OneMinus(None=@n10))) {minDefault=0.6}
  n12 = LinearInterpolate(
   A=LinearInterpolate(A=Constant3Vector() {constant=<rgba 0.1615 0.0277 0.0306 1>}, B=@n1.RGB, Alpha=Multiply(A=@n8, B=@n11))
   B=@n1.RGB
   Alpha=@n1.RGB
  )
  n13 = Multiply(A=Desaturation(None=@n12)) {constB=0.5}
  n14 = Saturate(None=Multiply(A=OneMinus(None=@n13)) {constB=0.8})
  FunctionOutput(None=SetMaterialAttributes!(
    MaterialAttributes=MaterialFunctionCall() {materialFunction=/VehicleCosmetics/Materials/MaterialFunctions/MF_Vehicle_Body_Base}.Result
    BaseColor=@n12
    Roughness=OneMinus(None=@n14)
   ) {attributeSetTypes [#1, #2]})
 ]}
