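# Clear-coat vehicle body material for the ZRF Aglet01 skin, set up for skeletal meshes and pixel animation.
# The MF_Skin_ZRF_Aglet01 function builds the material attributes, which feed the shared MF_Vehicle_FNPostFX function.
# Its clear-coat bottom normal is exposed through a named reroute for the ClearCoatNormal custom output.
# A separate Apply_Imperfections static switch branch runs the same attributes through MF_Imperfection, with no consumer shown.
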Material /VehicleCosmetics/Bodies/ZRF/Skins/Aglet01/Mat_Skin_ZRF_Aglet01
graph {
 properties {
  bHasPixelAnimation=true
  bUseMaterialAttributes=true
  bUsedWithMeshDeformer=true
  bUsedWithSkeletalMesh=true
  shadingModel=MSM_ClearCoat
  usedShadingModels=Clear Coat
 }
 nodes [
  ClearCoatNormalCustomOutput!(None=NamedRerouteUsage())
  n1 = MaterialFunctionCall() {materialFunction=&/MF_Skin_ZRF_Aglet01}
  MaterialFunctionCall!(In=@n1.Result) {materialFunction=/VehicleCosmetics/Materials/MaterialFunctions/MF_Vehicle_FNPostFX}
  NamedRerouteDeclaration(None=GetMaterialAttributes(None=@n1.Result) {attributeGetTypes [#1]}.ClearCoatBottomNormal) {name=BodyBase_BottomNormal, nodeColor=<rgba 0 1 0.3672 1>}
  StaticSwitchParameter(True=MaterialFunctionCall!(MaterialAttributes=@n1.Result) {materialFunction=/VehicleCosmetics/Materials/ExtraFunctions/MF_Imperfection}.Imperfection_MA) {group=Imperfection, parameterName=Apply_Imperfections, sortPriority=-1}
 ]
}
